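# Sound cue for searching a grill loot container, using the prop-search attenuation and loot-collect sound class.
# A mixer layers a random grill search sound at full volume, a quieter random cover sound, and a half-volume, one-second-delayed replay through the same search branch.
# Modulators vary pitch and volume on both random branches, with cover pitch only shifted upward.
# A concurrency override prevents new instances once the limit is reached.
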SoundCue /Game/Sounds/Foley_Loot/Containers/Grill/Grill_Search_Cue
root=Mixer_0
nodes [
 {id=Delay_22, children [Modulator_1], properties {delayMax=1, delayMin=1}}
 {id=Mixer_0, children [Modulator_1,Modulator_2,Delay_22], properties {inputVolume [1,0.4,0.5]}}
 {id=Modulator_1, children [Random_27], properties {pitchMax=1.07, pitchMin=0.92, volumeMax=1.07, volumeMin=0.92}}
 {id=Modulator_2, children [Random_28], properties {pitchMax=1.2, pitchMin=1, volumeMax=1.07, volumeMin=0.92}}
 {id=Random_27, children [WavePlayer_178, WavePlayer_179, WavePlayer_183], properties {weights [1,1,1]}}
 {id=Random_28, children [WavePlayer_184, WavePlayer_185, WavePlayer_186], properties {weights [1,1,1]}}
 {id=WavePlayer_178, properties {soundWaveAssetPtr=&/Grill_Search02}}
 {id=WavePlayer_179, properties {soundWaveAssetPtr=&/Grill_Search03}}
 {id=WavePlayer_183, properties {soundWaveAssetPtr=&/Grill_Search09}}
 {id=WavePlayer_184, properties {soundWaveAssetPtr=&/Grill_Cover_Search01}}
 {id=WavePlayer_185, properties {soundWaveAssetPtr=&/Grill_Cover_Search06}}
 {id=WavePlayer_186, properties {soundWaveAssetPtr=&/Grill_Cover_Search07}}
]
properties {
 attenuationSettings=/Game/Sounds/Attenuation/Fort_Prop_Search_Att
 bOverrideConcurrency=true
 maxDistance=1300
 soundClassObject=&Fort_Loot_Collect
 volumeMultiplier=1
 concurrencyOverrides {resolutionRule=PreventNew}
}
